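# Material function that computes a positional offset for creepy floating, swinging movement on the prototype flotation balloon weapon.
# Curl noise at the object position, shifted by phase-offset sines, drives the overall float.
# That noise plus time-scrolled world-position curl noise forms rope sway, masked by inverted bounds height and red vertex color.
# Both add to a normalized pivot-relative vector; X is reused for Y, and output Z is clamped non-negative.
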MaterialFunction /Game/Weapons/Prototype/Flotation_Balloon/Material/MF_CreepyBalloonMovement
graph {nodes [
  n1 = MaterialFunctionCall(Float3=Subtract(
    A=WorldPosition().XYZ
    B=MaterialFunctionCall() {materialFunction=&ObjectPivotPoint}.Object Pivot Location
   )) {materialFunction=&BreakOutFloat3Components}
  n2 = Add(A=@n1.B, B=ScalarParameter() {defaultValue=12, parameterName=AdjsutPivot})
  n3 = Time()
  n4 = FunctionInput(Preview=ScalarParameter() {defaultValue=0.02, parameterName=FloatSpeed}) {bUsePreviewValueAsDefault=true, inputName=FloatSpeed_1, inputType=FunctionInput_Scalar}
  n5 = Sine(None=Multiply(A=@n3, B=@n4))
  n6 = Sine(None=Multiply(A=Add(A=@n3) {constB=0.25}, B=@n4))
  n7 = Sine(None=Multiply(A=Add(A=@n3) {constB=0.5}, B=@n4))
  n8 = VectorNoise(World Position=Add(
    A=ObjectPositionWS()
    B=MaterialFunctionCall(X=@n5, Y=@n6, Z=@n7) {materialFunction=&MakeFloat3}.Result
   )) {noiseFunction=VNF_CurlALU}
  n9 = FunctionInput(Preview=ScalarParameter() {defaultValue=2, parameterName=Float_OFfset_Amt}) {bUsePreviewValueAsDefault=true, inputName=FloatingMovement_Amt_1, inputType=FunctionInput_Scalar}
  n10 = ComponentMask(None=Subtract(
    A=WorldPosition().XYZ
    B=MaterialFunctionCall() {materialFunction=&ObjectPivotPoint}.Object Pivot Location
   )) {b=true}
  n11 = FunctionInput(Preview=ScalarParameter() {defaultValue=80, parameterName=RopeEndOffset}) {bUsePreviewValueAsDefault=true, inputName=RopeEndOffset_1, inputType=FunctionInput_Scalar}
  n12 = MaterialFunctionCall() {materialFunction=&ObjectLocalBounds}
  n13 = OneMinus(None=Divide(A=Add(A=@n10, B=@n11), B=ComponentMask(None=@n12.Local Bounds Size) {b=true}))
  n14 = Constant() {r=64}
  n15 = MaterialFunctionCall(Float3=Divide(
    A=WorldPosition().XYZ
    B=FunctionInput(Preview=@n14) {bUsePreviewValueAsDefault=true, inputName=CurlNoiseScale, inputType=FunctionInput_Scalar}
   ) {constB=128}) {materialFunction=&BreakOutFloat3Components}
  n16 = Time()
  n17 = FunctionInput(Preview=ScalarParameter() {defaultValue=0.075, parameterName=SwingSpeed}) {bUsePreviewValueAsDefault=true, inputName=SwingSpeed_1, inputType=FunctionInput_Scalar}
  n18 = Add(A=@n15.R, B=Multiply(A=@n16, B=@n17))
  n19 = Add(A=@n15.G, B=Multiply(A=Add(A=@n16) {constB=0.25}, B=@n17))
  n20 = Add(A=@n15.B, B=Multiply(A=Add(A=@n16) {constB=0.75}, B=@n17))
  n21 = Multiply(
   A=Add(
    A=ObjectPositionWS()
    B=MaterialFunctionCall(X=@n18, Y=@n19, Z=@n20) {materialFunction=&MakeFloat3}.Result
   )
   B=FunctionInput(Preview=ScalarParameter() {defaultValue=2, parameterName=No-RIpples}) {bUsePreviewValueAsDefault=true, inputName=No of Ripples_1, inputType=FunctionInput_Scalar}
  )
  n22 = MaterialFunctionCall(Float3=Add(A=@n8, B=VectorNoise(World Position=@n21) {noiseFunction=VNF_CurlALU})) {materialFunction=&BreakOutFloat3Components}
  n23 = Multiply(
   A=Multiply(A=Clamp(None=@n13), B=VertexColor().R)
   B=MaterialFunctionCall(X=@n22.R, Y=@n22.R, Z=Clamp(None=@n22.B)) {materialFunction=&MakeFloat3}.Result
  )
  n24 = FunctionInput(Preview=ScalarParameter() {defaultValue=2, parameterName=Movement_Strength}) {bUsePreviewValueAsDefault=true, inputName=MovementStrength_1, inputType=FunctionInput_Scalar}
  n25 = Add(
   A=Normalize(VectorInput=MaterialFunctionCall(X=@n1.R, Y=@n1.R, Z=@n2) {materialFunction=&MakeFloat3}.Result)
   B=Add(A=Multiply(A=@n8, B=@n9), B=Multiply(A=@n23, B=@n24))
  )
  n26 = ComponentMask(None=@n25) {b=true}
  FunctionOutput(None=AppendVector(A=ComponentMask(None=@n25) {g=true, r=true}, B=Clamp(None=@n26) {maxDefault=1000})) {outputName=Result_1}
 ]}
