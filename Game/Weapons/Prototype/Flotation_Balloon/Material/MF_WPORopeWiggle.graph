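# Material function that generates a world position offset wiggle for a rope, filed under the prototype Flotation Balloon weapon.
# It sums three time-driven sine waves along UV V, scaled by object X scale, each with its own period, speed and Y/Z amplitude.
# Vertex color red times inverted green masks the motion, scaled by WiggleStrength.
# A SphereMask on a sine of V further shapes it, with hardness driven by object radius relative to MaxLength.
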MaterialFunction /Game/Weapons/Prototype/Flotation_Balloon/Material/MF_WPORopeWiggle
graph {nodes [
  n1 = Multiply(
   A=ComponentMask(None=TextureCoordinate()) {g=true}
   B=MaterialFunctionCall() {materialFunction=/Engine/Functions/Engine_MaterialFunctions02/WorldPositionOffset/ObjectScale}.Scale X
  )
  n2 = ScalarParameter() {defaultValue=1, parameterName=TimeScale}
  n3 = Multiply(A=Time(), B=Multiply(A=Constant() {r=0.1}, B=@n2)) {constB=0.1}
  n4 = Multiply(A=Sine(None=Add(A=@n1, B=@n3)) {period=0.5}, B=Constant3Vector() {constant=<rgba 0 0 0.1 0>})
  n5 = Multiply(
   A=ComponentMask(None=TextureCoordinate()) {g=true}
   B=MaterialFunctionCall() {materialFunction=&ObjectScale}.Scale X
  )
  n6 = Multiply(A=Time(), B=@n2)
  n7 = Multiply(A=Sine(None=Add(A=@n5, B=@n6)) {period=4}, B=Constant3Vector() {constant=<rgba 0 2 0.5 0>})
  n8 = ComponentMask(None=TextureCoordinate()) {g=true}
  n9 = MaterialFunctionCall() {materialFunction=&ObjectScale}
  n10 = Time()
  n11 = Multiply(A=@n2, B=Constant() {r=2})
  n12 = Sine(None=Add(A=Multiply(A=@n8, B=@n9.Scale X), B=Multiply(A=@n10, B=@n11) {constB=2})) {period=6}
  n13 = Constant3Vector() {constant=<rgba 0 0 2 0>}
  n14 = VertexColor()
  n15 = Multiply(A=Add(A=Add(A=@n4, B=@n7), B=Multiply(A=@n12, B=@n13)), B=Multiply(A=@n14.R, B=OneMinus(None=@n14.G)))
  n16 = ScalarParameter() {defaultValue=1, parameterName=WiggleStrength}
  n17 = Sine(None=ComponentMask(None=TextureCoordinate()) {g=true}) {period=2}
  n18 = Constant() {r=0.5}
  n19 = ScalarParameter() {defaultValue=0.5472, parameterName=SphereRadius}
  n20 = ObjectRadius()
  n21 = ScalarParameter() {defaultValue=1000, parameterName=MaxLength}
  n22 = Clamp(None=LinearInterpolate(Alpha=Divide(A=@n20, B=@n21)))
  FunctionOutput(None=Multiply(A=Multiply(A=@n15, B=@n16), B=SphereMask(A=@n17, B=@n18, Radius=@n19, Hardness=@n22) {attenuationRadius=0.5, hardnessPercent=50}))
 ]}
