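# Subsurface material for the Flotation Balloon gadget, usable on skeletal meshes.
# An FX mask texture separates tinted balloon, Fresnel-driven bag plastic and subsurface glow regions.
# A stencil decal is rotated and offset separately per UV half, then scaled and composited over base color.
# World position offset blends rope wiggle with creepy balloon movement by parameter, while mobile and low quality zero out subsurface color and opacity.
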Material /Game/Weapons/Prototype/Flotation_Balloon/Material/Flotation_Balloon
graph {
 properties {
  bUseMaterialAttributes=true
  bUsedWithMeshDeformer=true
  bUsedWithSkeletalMesh=true
  pixelDepthOffsetMode=PDOM_Legacy
  shadingModel=MSM_Subsurface
  usedShadingModels=Subsurface
 }
 nodes [
  n1 = TextureSampleParameter2D() {parameterName=Diffuse, texture=/Game/Weapons/Prototype/Flotation_Balloon/Texture/T_Gadget_FlotationBalloon_D}
  n2 = TextureSample() {
   samplerSource=SSM_Wrap_WorldGroupSettings
   samplerType=SAMPLERTYPE_Masks
   texture=/Game/Weapons/Prototype/Flotation_Balloon/Texture/T_Gadget_FlotationBalloon_FX
  }
  n3 = ScalarParameter() {defaultValue=-0.25, parameterName=rerer}
  n4 = ScalarParameter() {defaultValue=0.8, group=SS Shader Model, parameterName=gdsgfdsf}
  n5 = Fresnel(ExponentIn=ScalarParameter() {defaultValue=1, group=SS Shader Model, parameterName=Bag Fresnel Plastic Exponent})
  n6 = Multiply(A=@n2.B, B=Saturate(None=LinearInterpolate(A=@n3, B=@n4, Alpha=@n5)))
  n7 = LinearInterpolate(A=@n1.RGB, B=VectorParameter() {parameterName=bag plastic color, defaultValue=<rgba 0.2549 0.4004 0.5625 1>}.RGB, Alpha=@n6)
  n8 = VectorParameter() {parameterName=Balloon Color, defaultValue=<rgba 1 0 0 0>}
  n9 = Multiply(A=@n1.RGB, B=@n8.RGB)
  n10 = TextureCoordinate()
  n11 = Rotator(Coordinate=@n10, Time=ScalarParameter() {defaultValue=6.28, parameterName=RotateStencil_1})
  n12 = AppendVector(
   A=ScalarParameter() {defaultValue=0.24, parameterName=StencilOffset_X_1}
   B=ScalarParameter() {defaultValue=-0.18, parameterName=StencilOffset_Y_1}
  )
  n13 = Rotator(Coordinate=@n10, Time=ScalarParameter() {defaultValue=-6.28, parameterName=RotateStencil})
  n14 = AppendVector(
   A=ScalarParameter() {defaultValue=0.26, parameterName=StencilOffset_X}
   B=ScalarParameter() {defaultValue=0.07, parameterName=StencilOffset_Y}
  )
  n15 = Add(A=ComponentMask(None=TextureCoordinate()) {g=true}) {constB=-0.5}
  n16 = MaterialFunctionCall(
   UVs=LinearInterpolate(A=Subtract(A=@n11, B=@n12) {constA=0.5, constB=0.5}, B=Subtract(A=@n13, B=@n14) {constA=0.5, constB=0.5}, Alpha=Ceil(None=@n15))
   Texture Scale=AppendVector(A=ScalarParameter() {defaultValue=0.25, parameterName=StencilScale_X}, B=ScalarParameter() {defaultValue=0.31, parameterName=StencilScale_Y})
  ) {materialFunction=&ScaleUVsByCenter}
  n17 = TextureSampleParameter2D(UVs=@n16.UVs) {parameterName=Stencil, texture=/Game/Environments/Props/BirthdayBalloons/Textures/T_NoDecal}
  n18 = TextureSampleParameter2D() {parameterName=SpecularMasks, samplerType=SAMPLERTYPE_Masks, texture=/Game/Weapons/Prototype/Flotation_Balloon/Texture/T_Balloon_S}
  n19 = ScalarParameter() {defaultValue=0.05, group=SS Shader Model, parameterName=SS Inner Strength}
  n20 = ScalarParameter() {defaultValue=1.5, group=SS Shader Model, parameterName=SS Outer Strength}
  n21 = Fresnel(ExponentIn=ScalarParameter() {defaultValue=4, group=SS Shader Model, parameterName=SS Fresnel Exponent})
  n22 = Saturate(None=Multiply(A=@n8.RGB, B=ScalarParameter() {defaultValue=0.5, parameterName=balloon ss color modifier}))
  n23 = MakeMaterialAttributes!(
   BaseColor=LinearInterpolate(A=LinearInterpolate(A=@n7, B=@n9, Alpha=@n2.G), B=@n17.RGB, Alpha=Multiply(A=@n17.R, B=@n16.0-1 mask))
   Metallic=@n18.G
   Specular=@n18.G
   Roughness=LinearInterpolate(A=LinearInterpolate(Alpha=@n18.B) {constA=0.1}, B=ScalarParameter() {defaultValue=0.1, parameterName=bag roughness}, Alpha=@n6)
   Anisotropy
   EmissiveColor
   Opacity=LinearInterpolate(B=LinearInterpolate(A=@n19, B=@n20, Alpha=@n21), Alpha=@n2.R) {constB=0}
   OpacityMask
   Normal=TextureSampleParameter2D() {parameterName=Normals, samplerType=SAMPLERTYPE_Normal, texture=/Game/Weapons/Prototype/Flotation_Balloon/Texture/T_Balloon_N}.RGB
   Tangent
   WorldPositionOffset=LinearInterpolate(
    A=MaterialFunctionCall() {materialFunction=&/MF_WPORopeWiggle}.Result
    B=MaterialFunctionCall!(No of Ripples_1, MovementStrength_1, RopeEndOffset_1, FloatSpeed_1, FloatingMovement_Amt_1, CurlNoiseScale, SwingSpeed_1) {materialFunction=&/MF_CreepyBalloonMovement}.Result_1
    Alpha=ScalarParameter() {parameterName=UseCreepyBalloonMovement}
   )
   SubsurfaceColor=LinearInterpolate(B=LinearInterpolate(A=@n22, B=@n8.RGB, Alpha=@n21), Alpha=@n2.R)
   ClearCoat
   ClearCoatRoughness
   AmbientOcclusion
   Refraction
   CustomizedUVs_0
   CustomizedUVs_1
   CustomizedUVs_2
   CustomizedUVs_3
   CustomizedUVs_4
   CustomizedUVs_5
   CustomizedUVs_6
   CustomizedUVs_7
   PixelDepthOffset
   ShadingModel
   Displacement
  )
  n24 = Constant()
  n25 = SetMaterialAttributes!(MaterialAttributes=@n23, Subsurface Color=@n24, Opacity=@n24) {attributeSetTypes [#1, #2]}
  n26 = ShadingPathSwitch(Default=@n23, Mobile=@n25) {desc=Migrated from FeatureLevelSwitch}
  QualitySwitch(Default=@n26, Low=@n25, High=@n26, Medium=@n26)
 ]
}
