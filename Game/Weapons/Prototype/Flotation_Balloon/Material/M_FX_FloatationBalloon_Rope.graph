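# Masked, lit particle material for the prototype flotation balloon's rope, using the balloon textures and dithered opacity.
# An offset graph bends the rope toward a Final Offset Position along a contrast-adjusted UV1 gradient, driven by the Bend dynamic parameter.
# It adds three layered, scrolling sine waves, scaled by object scale and Wiggle Strength.
# A sphere mask centers the wiggle mid-rope, with hardness derived from object radius against MaxLength.
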Material /Game/Weapons/Prototype/Flotation_Balloon/Material/M_FX_FloatationBalloon_Rope
graph {
 properties {blendMode=BLEND_Masked, ditherOpacityMask=true, pixelDepthOffsetMode=PDOM_Legacy, usedShadingModels=Default Lit}
 nodes [
  n1 = TextureCoordinate() {coordinateIndex=1}
  n2 = Power(Base=OneMinus(None=ComponentMask(None=@n1) {g=true}), Exp=ScalarParameter() {defaultValue=1.6121, parameterName=Gradient Contrast})
  n3 = DynamicParameter() {defaultValue=<rgba 0 0 1 1>, paramNames [Bend, Wiggle Strength, Param3, Param4]}
  n4 = ComponentMask(None=TextureCoordinate()) {g=true}
  n5 = MaterialFunctionCall() {materialFunction=/Engine/Functions/Engine_MaterialFunctions02/WorldPositionOffset/ObjectScale}
  n6 = Time()
  n7 = Sine(None=Add(A=Multiply(A=@n4, B=@n5.Scale X), B=Multiply(A=@n6) {constB=0.1})) {period=0.5}
  n8 = Constant3Vector() {constant=<rgba 0 0 0.1 0>}
  n9 = ComponentMask(None=TextureCoordinate()) {g=true}
  n10 = MaterialFunctionCall() {materialFunction=&ObjectScale}
  n11 = Time()
  n12 = Sine(None=Add(A=Multiply(A=@n9, B=@n10.Scale X), B=Multiply(A=@n11))) {period=4}
  n13 = Constant3Vector() {constant=<rgba 0 2 0.5 0>}
  n14 = TextureCoordinate()
  n15 = Add(
   A=Multiply(
    A=ComponentMask(None=@n14) {g=true}
    B=MaterialFunctionCall() {materialFunction=&ObjectScale}.Scale X
   )
   B=Multiply(A=Time()) {constB=2}
  )
  n16 = Add(
   A=Add(A=Multiply(A=@n7, B=@n8), B=Multiply(A=@n12, B=@n13))
   B=Multiply(A=Sine(None=@n15) {period=6}, B=Constant3Vector() {constant=<rgba 0 0 2 0>})
  )
  n17 = ComponentMask(None=TextureCoordinate()) {g=true}
  n18 = Constant() {r=0.5}
  n19 = ScalarParameter() {defaultValue=0.5, parameterName=SphereRadius}
  n20 = ObjectRadius()
  n21 = ScalarParameter() {defaultValue=750, parameterName=MaxLength}
  n22 = Clamp(None=LinearInterpolate(Alpha=Divide(A=@n20, B=@n21)))
  Add(
   A=LinearInterpolate(B=VectorParameter() {parameterName=Final Offset Position, defaultValue=<rgba 36.5359 -65 -68.7991 0>}.RGB, Alpha=Multiply(A=@n2, B=@n3.Bend)) {constB=0}
   B=Multiply(
    A=Multiply(A=@n16, B=@n3.Wiggle Strength)
    B=SphereMask(A=@n17, B=@n18, Radius=@n19, Hardness=@n22) {attenuationRadius=0.5, hardnessPercent=50}
   )
  )
  LinearInterpolate(Alpha=TextureSample() {samplerType=SAMPLERTYPE_Masks, texture=/Game/Weapons/Prototype/Flotation_Balloon/Texture/T_Balloon_S}.B) {constA=0.1}
  ParticleColor()
  TextureSample() {samplerType=SAMPLERTYPE_Normal, texture=/Game/Weapons/Prototype/Flotation_Balloon/Texture/T_Balloon_N}
  TextureSample() {texture=/Game/Weapons/Prototype/Flotation_Balloon/Texture/T_Balloon_D}
 ]
}
